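; TrafficHat NPC pawn Blueprint that adds an overhead health and shield bar and a team-only indicator when converted (hired).
; The server packs health changes into HealthInfo, whose OnRep updates the AIBoss status widget for nearby local players.
; Clients tick the widget component only while visible and use a timer to shift it by player distance.
; Death of a converted NPC empties the bar, then hides it after a configurable delay.
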(event Custom|EnableHiredTandemHealthBar
  (Variables|InternalData|SetIsHealthCheatEnabled true)
  (Networking|SwitchHasAuthority
    (:Authority
      (Health|BindEventtoOnPawnHealthChanged (delegate Custom|PawnHealthChanged)))
    (:Remote
      (CallFunction|StartUpdatingHealthWidget))))

(event Custom|EnableTandemHealthBar
  (Variables|HealthBar|SetIsTandemHealthEnabled true)
  (if (ScalableFloat|GetValueasBool (Variables|HealthBar|GetShowHealth))
    (Networking|SwitchHasAuthority
      (:Authority
        (Health|BindEventtoOnPawnHealthChanged (delegate Custom|PawnHealthChanged)))
      (:Remote
        (CallFunction|StartUpdatingHealthWidget)))))

(event Custom|StartHealthBar
  (if (ScalableFloat|GetValueasBool (Variables|HealthBar|GetShowHealth))
    (Networking|SwitchHasAuthority
      (:Authority
        (Health|BindEventtoOnPawnHealthChanged (delegate Custom|PawnHealthChanged)))
      (:Remote
        (CallFunction|StartUpdatingHealthWidget)))))

(event Custom|PawnHealthChanged
  (if (Variables|InternalData|GetIsConverted)
    _
    (elif (Variables|HealthBar|GetIsTandemHealthEnabled)
      _))
  (Variables|InternalData|SetHealthInfo (Utilities|Struct|MakeStructNPCHealthInfo (Math|Float|SafeDivide (FortPawn|GetHealth) (FortPawn|GetMaxHealth)) (Math|Float|SafeDivide (FortPawn|GetShield) (FortPawn|GetMaxShield)) (FortPawn|GetMaxHealth) (FortPawn|GetMaxShield))))

(event EventEndPlay
  (params (EndPlayReason :type EEndPlayReason%20Enum :default "Destroyed"))
  (if (Variables|InternalData|GetIsHealthCheatEnabled)
    _
    (elif (ScalableFloat|GetValueasBool (Variables|HealthBar|GetShowHealth))
      _))
  (CallFunction|StopUpdatingHealthWidget))

(event EventOnDeathPlayEffects
  (params (Damage :type Float%20%28single-precision%29) (DamageTags :type Gameplay%20Tag%20Container%20Structure%20%28by%20ref%29) (Momentum :type Vector :default "0, 0, 0") (HitInfo :type Hit%20Result%20Structure%20%28by%20ref%29) (InstigatedBy :type Fort%20Pawn%20Object%20Reference) (DamageCauser :type Actor%20Object%20Reference) (EffectContext :type Gameplay%20Effect%20Context%20Handle%20Structure))
  (|CallParentFunction Damage DamageTags Momentum HitInfo InstigatedBy DamageCauser EffectContext)
  (if (Variables|InternalData|GetIsHealthCheatEnabled)
    _
    (elif (ScalableFloat|GetValueasBool (Variables|HealthBar|GetShowHealth))
      _))
  (if (Variables|InternalData|GetIsConverted)
    (CallFunction|NPCSetupHealthBaronDeath)))

(event Custom|StartUpdatingHealthWidget
  ;[1 1 1] Health Bar
  (bind _npcstatuswidget (Variables|Default|GetNPCStatusWidget)
    (:then
      (bind _asnpcstatus_widget_aiboss (Utilities|Casting|CastToNPCStatusWidget_AIBoss (UserInterface|GetUserWidgetObject _npcstatuswidget)))
      (Utilities|FlowControl|Sequence
        (:then_0
          (Components|Tick|SetComponentTickEnabled (Variables|Default|GetNPCStatusWidget) (Widget|IsVisible _asnpcstatus_widget_aiboss)))
        (:then_1
          (Default|BindEventtoVisibilityChanged _asnpcstatus_widget_aiboss (delegate Custom|WidgetVisibilityChanged))
          (bind _returnvalue (Utilities|Time|SetTimerbyEvent (EventDispatchers|CreateEvent :!event Custom|UpdateWidgetOffset) 0.1 true))
          (Variables|Default|SetUpdateOffsetTimerHandle _returnvalue))))
    (:else)))

(event Collision|EventActorBeginOverlap
  (params (OtherActor :type Actor%20Object%20Reference))
  (|CallParentFunction OtherActor))

(event Custom|WidgetVisibilityChanged
  (params (bVisible :type Boolean))
  ;[1 1 1] Health Bar
  (Components|Tick|SetComponentTickEnabled (Variables|Default|GetNPCStatusWidget) bVisible))

(event EventTick
  (params (DeltaSeconds :type Float%20%28single-precision%29))
  (|CallParentFunction DeltaSeconds) ; This node is disabled and will not be called.\nDrag off pins to build functionality.
)

(event Custom|StopUpdatingHealthWidget
  ;[1 1 1] Health Bar
  (bind _npcstatuswidget (Variables|Default|GetNPCStatusWidget)
    (:then
      (bind _asnpcstatus_widget_aiboss (Utilities|Casting|CastToNPCStatusWidget_AIBoss (UserInterface|GetUserWidgetObject _npcstatuswidget)))
      (Default|UnbindEventfromVisibilityChanged _asnpcstatus_widget_aiboss (delegate Custom|WidgetVisibilityChanged))
      (Utilities|Time|ClearandInvalidateTimerbyHandle (Variables|Default|GetUpdateOffsetTimerHandle)))
    (:else)))

(event Custom|NPCSetupHealthBarOnDeath
  (bind (_healthpercent_2_7edada384c7908ea98debd9958a249de _shieldpercent_5_b79c63b64aacf275a1621a9315964dcd _maxhealth_6_7d7bef7841b13fb96d2578989cf135bb _maxshield_9_f923e4444e2e476c56fb0e9d9d2f7846) (Utilities|Struct|BreakStructNPCHealthInfo (Variables|InternalData|GetHealthInfo)))
  ;[1 1 1] Health Bar Handling on Death
  (if (> (Variables|HealthBar|GetNPCHealthBarHideTimeAfterDeath) 0)
    (bind _npcstatuswidget (Variables|Default|GetNPCStatusWidget)
      (:then
        (bind _asnpcstatus_widget_aiboss (Utilities|Casting|CastToNPCStatusWidget_AIBoss (UserInterface|GetUserWidgetObject _npcstatuswidget)))
        (Class|NPCStatusWidgetAIBoss|EventUpdateStatus _asnpcstatus_widget_aiboss 0.0 0.0 _maxhealth_6_7d7bef7841b13fb96d2578989cf135bb _maxshield_9_f923e4444e2e476c56fb0e9d9d2f7846 false false)
        (Utilities|FlowControl|Delay (Variables|HealthBar|GetNPCHealthBarHideTimeAfterDeath)))
      (:else))
    (else
      _))
  (bind _npcstatuswidget_2 (Variables|Default|GetNPCStatusWidget)
    (:then
      (Development|SetHiddeninGame _npcstatuswidget_2 true false))
    (:else)))

(event Custom|NPCTandemDisableHealthBar
  ;[1 1 1] Debug For Captures
  (Variables|InternalData|SetNPCStatusWidgetHealthBarDisabled true)
  (bind _npcstatuswidget (Variables|Default|GetNPCStatusWidget)
    (:then
      (bind _asnpcstatus_widget_aiboss (Utilities|Casting|CastToNPCStatusWidget_AIBoss (UserInterface|GetUserWidgetObject _npcstatuswidget)))
      (Widget|SetVisibility _asnpcstatus_widget_aiboss "Hidden"))
    (:else)))

(event Custom|NPCTandemEnableHealthBars
  ;[1 1 1] Debug For Captures
  (Variables|InternalData|SetNPCStatusWidgetHealthBarDisabled false))

(event Custom|UpdateWidgetOffset
  (bind _returnvalue (Fortnite|GetFirstLocalFortPlayerController true))
  (bind _returnvalue_2 (Pawn|GetControlledPawn _returnvalue))
  ;[1 1 1] UpdateWidgetOffset
  (Utilities|IsValid _returnvalue_2
    (:Is%20Valid
      (Variables|/Script/UMG.Widget|SetRenderTransform_Translation_X (Utilities|Struct|MakeWidgetTransform (Math|Vector2D|MakeVector2D 0.0 (Math|Float|Lerp 0.0 -25.0 (Math|Float|Clamp(Float) (Math|Float|NormalizetoRange (Math|Vector|Distance(Vector) (Transformation|GetActorLocation) (Transformation|GetActorLocation _returnvalue_2)) 0.0 5e+03))))) (UserInterface|GetUserWidgetObject (Variables|Default|GetNPCStatusWidget))))))

(fn HandlePawnConverted
  (params (InstigatorPawn :type Fort%20Pawn%20Object%20Reference) (ConvertedPawn :type Fort%20Pawn%20Object%20Reference))
  (|CallParentFunction InstigatorPawn ConvertedPawn)
  (Variables|InternalData|SetIsConverted true)
  (Variables|InternalData|SetLeader%20Team (Team|GetActorTeam InstigatorPawn))
  (Utilities|FlowControl|Sequence
    (:then_0
      (CallFunction|StartHealthBar))
    (:then_1
      (CallFunction|UpdateConvertIndicator))))

(fn HandlePawnUnconverted
  (params (UnconvertedPawn :type Fort%20Pawn%20Object%20Reference))
  (|CallParentFunction UnconvertedPawn)
  (Variables|InternalData|SetIsConverted false)
  (Variables|InternalData|SetLeader%20Team 0)
  (CallFunction|UpdateConvertIndicator))

(fn OnRep_HealthInfo
  (CallFunction|UpdatePawnHealthHUD))

(fn UpdateConvertIndicator
  (bind _convertcomponent (Variables|FortPlayerPawn|GetConvertComponent))
  (Networking|SwitchHasAuthority
    (:Remote
      (if (Math|Byte|Equal(Byte) (Team|GetActorTeam (Fortnite|GetFirstLocalFortPlayerController true)) (Variables|InternalData|GetLeader%20Team))
        (Convert|ShowIndicatorMarker _convertcomponent)
        (else
          (Convert|HideIndicatorMarker _convertcomponent))))))

(fn UpdatePawnHealthHUD
  (bind (_healthpercent_2_7edada384c7908ea98debd9958a249de _shieldpercent_5_b79c63b64aacf275a1621a9315964dcd _maxhealth_6_7d7bef7841b13fb96d2578989cf135bb _maxshield_9_f923e4444e2e476c56fb0e9d9d2f7846) (Utilities|Struct|BreakStructNPCHealthInfo (Variables|InternalData|GetHealthInfo)))
  (Networking|SwitchHasAuthority
    (:Remote
      (bind _array_element (Utilities|Array|ForEachLoopwithBreak (Fortnite|GetLocalFortPlayerControllers true))
        (:LoopBody
          (if (Math|Boolean|ANDBoolean (< (Class|AbilityFunctions|GetDistanceBetweenActors (Pawn|GetControlledPawn _array_element) self) (Variables|HealthBar|GetNPCMaxDistanceFromDamageINstigatorToDrawHealthBar)) (not (Variables|Health|GetbIsDying)) (not (Variables|InternalData|GetNPCStatusWidgetHealthBarDisabled)))
            (bind _npcstatuswidget (Variables|Default|GetNPCStatusWidget)
              (:then
                (bind _asnpcstatus_widget_aiboss (Utilities|Casting|CastToNPCStatusWidget_AIBoss (UserInterface|GetUserWidgetObject _npcstatuswidget)))
                (Class|NPCStatusWidgetAIBoss|EventUpdateStatus _asnpcstatus_widget_aiboss _healthpercent_2_7edada384c7908ea98debd9958a249de _shieldpercent_5_b79c63b64aacf275a1621a9315964dcd _maxhealth_6_7d7bef7841b13fb96d2578989cf135bb _maxshield_9_f923e4444e2e476c56fb0e9d9d2f7846 true false))
              (:else))))))))

(fn ConstructionScript
  (|CallParentFunction))
